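# Material function in the Figment environment's landscape set that blends only the Specular attribute of two material attribute inputs.
# It extracts Specular from Mat A and Mat B and linearly interpolates between them using a scalar Alpha input.
# The blended value is written back onto Mat A's attributes, so every other attribute passes through from Mat A and Mat B contributes nothing else.
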MaterialFunction /FigmentEnvironment/LM_8_1_1/Athena/Environments/Landscape/MaterialFunctions/MF_Figment_BlendMatSpec
graph {nodes [
  n1 = GetMaterialAttributes(None=FunctionInput() {description=0, inputName=Mat A, inputType=FunctionInput_MaterialAttributes}) {attributeGetTypes [#1]}
  n2 = GetMaterialAttributes(None=FunctionInput() {description=1, inputName=Mat B, inputType=FunctionInput_MaterialAttributes, sortPriority=1}) {attributeGetTypes [#1]}
  n3 = FunctionInput() {inputName=Alpha, inputType=FunctionInput_Scalar, sortPriority=2}
  FunctionOutput(None=SetMaterialAttributes!(MaterialAttributes=@n1.MaterialAttributes, Specular=LinearInterpolate(A=@n1.Specular, B=@n2.Specular, Alpha=@n3)) {attributeSetTypes [#1]})
 ]}
