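# Master landscape material for Athena terrain, blending painted grass, rock, arid, snow, ice, forest, jungle, farm and path layers.
# Material functions build each surface, combined by slope and height blends and tinted by biome colormaps and MPC_Landscape parameters.
# An RVT switch selects between this composite and runtime virtual texture samples.
# Quality and feature-level switches provide cheaper paths, while EnableQuickColorRender adds colormap and biome-mask render views.
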Material /FigmentEnvironment/S10/Landscape/M_Athena_Terrain_Master
graph {
 properties {
  bUseMaterialAttributes=true
  bUsedWithNanite=true
  pixelDepthOffsetMode=PDOM_Legacy
  refractionMethod=RM_IndexOfRefraction
  usedShadingModels=Default Lit
 }
 nodes [
  n1 = LandscapeLayerSample() {parameterName=Farm_Grass}
  Add(A=@n1)
  n2 = Add(
   A=Add(A=LandscapeLayerSample() {parameterName=Road, previewWeight=1}, B=LandscapeLayerSample() {parameterName=Mud, previewWeight=1})
   B=LandscapeLayerSample() {parameterName=Gravel, previewWeight=1}
  )
  Add(A=@n2, B=LandscapeLayerSample() {parameterName=NoTopo, previewWeight=1})
  n3 = RuntimeVirtualTextureSampleParameter() {materialType=BaseColor_Normal_Specular_Mask_YCoCg, parameterName=RVT, virtualTexture=/FigmentCore/Athena/Prototype/Terrain/RVT_Figment}
  n4 = Constant()
  n5 = Multiply(
   A=NamedRerouteUsage()
   B=MaterialFunctionCall() {materialFunction=/FigmentEnvironment/LM_8_1_1/Athena/WinterArt/Materials/MF_SharedWinter}.IceColor
  )
  n6 = Multiply(
   A=Constant3Vector() {constant=<rgba 0.3346 0.4223 0.5 0>}
   B=Clamp(None=MaterialFunctionCall!(UseLandscapeCoordsOnSwitch) {materialFunction=/FigmentEnvironment/LM_8_1_1/Athena/Environments/Landscape/MaterialFunctions/Snow/MF_Athena_Snow_01}.NormalDotCamera)
  )
  n7 = LandscapeLayerSample() {parameterName=Gravel}
  n8 = LandscapeLayerSample() {parameterName=Forest}
  n9 = Subtract(A=LandscapeLayerSample() {parameterName=Ice}, B=Add(A=Add(A=@n7, B=@n8), B=LandscapeLayerSample() {parameterName=Mud})) {desc=Mask out paths and dirt from emmisive}
  n10 = StaticSwitchParameter(True=@n4, False=LinearInterpolate(A=@n4, B=Multiply(A=@n5, B=@n6) {constB=0.5}, Alpha=Clamp(None=@n9))) {group=Utilities, parameterName=EnableQuickColorRender}
  n11 = Constant()
  n12 = FeatureLevelSwitch(Default=@n10, ES2=@n11, ES3_1=@n11, SM4_REMOVED=@n10, SM5=@n10, SM6=@n10)
  n13 = SetMaterialAttributes!(
   MaterialAttributes
   Base Color=@n3.BaseColor
   Specular=@n3.BaseColor
   Roughness=@n3.BaseColor
   Opacity Mask=@n3.BaseColor
   Normal=Transform(None=@n3.Normal) {transformSourceType=TRANSFORMSOURCE_World, transformType=TRANSFORM_Tangent}
   Emissive Color=@n12
  ) {attributeSetTypes [
    #1
    #2
    #3
    #4
    #5
    #6
   ]}
  n14 = MaterialFunctionCall(Transition Phase=LandscapeLayerSample() {parameterName=Jungle_Sand}, Height Texture=OneMinus(None=NamedRerouteUsage()), Contrast=Constant()) {materialFunction=&HeightLerp}
  n15 = TextureObject() {
   samplerType=SAMPLERTYPE_LinearColor
   texture=/FigmentEnvironment/LM_8_1_1/Athena/Environments/Landscape/Textures/Arid/T_Athena_Arid_Rock_01_M
  }
  n16 = MaterialFunctionCall!(In=@n15, UseRoughness=StaticBool()) {materialFunction=/FigmentEnvironment/LM_8_1_1/Athena/Environments/Landscape/MaterialFunctions/Arid/MF_Athena_Arid_Sand}
  n17 = MaterialFunctionCall!(WorldPosition) {materialFunction=/FigmentEnvironment/LM_8_1_1/Athena/Environments/Landscape/MaterialFunctions/Utilities/MF_Athena_Terrain_Coords}
  n18 = TextureSample(UVs=@n17.Result) {
   samplerSource=SSM_Wrap_WorldGroupSettings
   samplerType=SAMPLERTYPE_LinearColor
   texture=/FigmentEnvironment/LM_8_1_1/Athena/Environments/Landscape/Textures/T_Athena_Terrain_RockColor
  }
  n19 = TextureSampleParameter2D(UVs=@n17.Result) {
   parameterName=BiomeColormap
   samplerType=SAMPLERTYPE_LinearColor
   texture=/FigmentEnvironment/FN_900/Athena/Environments/Landscape/Textures/T_Figment_Terrain_CombinedColors_S9
  }
  n20 = Multiply(
   A=LinearInterpolate(A=@n18.RGB, B=@n19.RGB, Alpha=NamedRerouteUsage()) {desc=Uses 2 color maps for better accuracy on height blend. This saves having to use layers for Biomes}
   B=CollectionParameter() {collection=/FigmentEnvironment/LM_8_1_1/Athena/Environments/Landscape/MPC/MPC_Landscape, parameterName=ColormapDivisor}
  )
  n21 = LinearInterpolate(
   A=ComponentMask(None=CollectionParameter() {collection=&MPC_Landscape, parameterName=FarmRockTint}) {b=true, g=true, r=true}
   B=ComponentMask(None=CollectionParameter() {collection=&MPC_Landscape, parameterName=FarmGrassTint}) {b=true, g=true, r=true}
   Alpha=NamedRerouteUsage()
  )
  n22 = LandscapeLayerSample() {parameterName=Farm_Grass}
  n23 = NamedRerouteUsage()
  n24 = Multiply(A=LandscapeLayerSample() {parameterName=Arid, previewWeight=1}, B=NamedRerouteUsage())
  n25 = MaterialFunctionCall() {materialFunction=/FigmentEnvironment/LM_8_1_1/Athena/Environments/Landscape/MaterialFunctions/Standard/MF_Athena_RoadGravel}
  n26 = MaterialFunctionCall!(Gravel=AppendVector(A=@n25.Gravel, B=@n25.Gravel)) {materialFunction=/FigmentEnvironment/LM_8_1_1/Athena/Environments/Landscape/MaterialFunctions/Standard/MF_Athena_Rock_01}
  n27 = CollectionParameter() {collection=&MPC_Landscape, parameterName=Rock_Bottom_Brightness}
  n28 = CollectionParameter() {collection=&MPC_Landscape, parameterName=Rock_Top_Brightness}
  n29 = MaterialFunctionCall!(height_Offset, height_GradScale, Coords=WorldPosition().XYZ) {materialFunction=/FigmentEnvironment/LM_8_1_1/Athena/Environments/Landscape/MaterialFunctions/MF_Athena_WorldHeightGrad}
  n30 = MaterialFunctionCall() {materialFunction=/FigmentEnvironment/LM_8_1_1/Athena/Environments/Landscape/MaterialFunctions/Standard/MF_Athena_Grass_01}
  n31 = Constant() {r=1}
  n32 = MaterialFunctionCall!(Height=@n30.Height, Transition Multiplier=@n31, Height Multiplier=Constant() {r=0.1}) {materialFunction=/FigmentEnvironment/LM_8_1_1/Athena/Environments/Landscape/MaterialFunctions/Utilities/MF_Athena_SlopeBlend}
  MaterialExpressionMaterialFunctionCall_31 = MaterialFunctionCall!(
   Base Material=MaterialFunctionCall!(In=@n26.Result, Color=LinearInterpolate(A=@n27, B=@n28, Alpha=@n29.Result)) {materialFunction=/FigmentEnvironment/LM_8_1_1/Athena/Environments/Landscape/MaterialFunctions/MF_Athena_Material_BaseMultiply}.Result
   Top Material=@n30.Result
   Alpha=@n32.Alpha
  ) {materialFunction=/Engine/Functions/MaterialLayerFunctions/MatLayerBlend_Standard}
  MaterialExpressionReroute_65 = Reroute(None=@MaterialExpressionMaterialFunctionCall_31.Blended Material)
  MaterialExpressionNamedRerouteUsage_60 = NamedRerouteUsage()
  MaterialExpressionNamedRerouteUsage_39 = NamedRerouteUsage()
  MaterialExpressionNamedRerouteUsage_51 = NamedRerouteUsage()
  MaterialExpressionReroute_67 = Reroute(None=@n26.Result)
  n33 = LandscapeLayerBlend!(
   Layer Grass_Topology=@MaterialExpressionReroute_65
   Layer Arid=@MaterialExpressionNamedRerouteUsage_60
   Layer Grass_Only=@MaterialExpressionNamedRerouteUsage_39
   Height Grass_Only=@MaterialExpressionNamedRerouteUsage_51
   Layer Snow=@MaterialExpressionReroute_67
  ) {layers [
    {
     blendType=LB_AlphaBlend
     constHeightInput=0
     layerName=Grass_Topology
     previewWeight=1
     constLayerInput=<0 0 0>
     heightInput {expression=null, inputName=None, mask=0, maskA=0, maskB=0, maskG=0, maskR=0, outputIndex=0}
     layerInput {
      expression=&self:MaterialExpressionReroute_65
      inputName=None
      mask=0
      maskA=0
      maskB=0
      maskG=0
      maskR=0
      outputIndex=0
     }
    }
    {
     blendType=LB_AlphaBlend
     constHeightInput=0
     layerName=Arid
     previewWeight=1
     constLayerInput=<0 0 0>
     heightInput {expression=null, inputName=None, mask=0, maskA=0, maskB=0, maskG=0, maskR=0, outputIndex=0}
     layerInput {
      expression=&self:MaterialExpressionNamedRerouteUsage_60
      inputName=None
      mask=0
      maskA=0
      maskB=0
      maskG=0
      maskR=0
      outputIndex=0
     }
    }
    {
     blendType=LB_HeightBlend
     constHeightInput=0
     layerName=Grass_Only
     previewWeight=1
     constLayerInput=<0 0 0>
     heightInput {
      expression=&self:MaterialExpressionNamedRerouteUsage_51
      inputName=None
      mask=0
      maskA=0
      maskB=0
      maskG=0
      maskR=0
      outputIndex=0
     }
     layerInput {
      expression=&self:MaterialExpressionNamedRerouteUsage_39
      inputName=None
      mask=0
      maskA=0
      maskB=0
      maskG=0
      maskR=0
      outputIndex=0
     }
    }
    {
     blendType=LB_AlphaBlend
     constHeightInput=0
     layerName=Snow
     previewWeight=0
     constLayerInput=<0 0 0>
     heightInput {expression=null, inputName=None, mask=0, maskA=0, maskB=0, maskG=0, maskR=0, outputIndex=0}
     layerInput {
      expression=&self:MaterialExpressionReroute_67
      inputName=None
      mask=0
      maskA=0
      maskB=0
      maskG=0
      maskR=0
      outputIndex=0
     }
    }
   ]}
  n34 = LandscapeLayerSample() {parameterName=Grass_Only, previewWeight=1}
  n35 = NamedRerouteUsage()
  n36 = LandscapeLayerSample() {parameterName=Snow, previewWeight=1}
  n37 = NamedRerouteUsage()
  n38 = MaterialFunctionCall!(
   Mat A=MaterialFunctionCall(
    Mat A=MaterialFunctionCall(Mat A=@n33, Mat B=@n23, Alpha=@n24) {materialFunction=/FigmentEnvironment/LM_8_1_1/Athena/Environments/Landscape/MaterialFunctions/MF_Figment_BlendMatSpec}.Result
    Mat B=NamedRerouteUsage()
    Alpha=Multiply(A=@n34, B=@n35)
   ) {materialFunction=&MF_Figment_BlendMatSpec}.Result
   Mat B=@MaterialExpressionReroute_67
   Alpha=MaterialFunctionCall(Transition Phase=Multiply(A=@n36, B=@n37), Height Texture=NamedRerouteUsage()) {materialFunction=&HeightLerp}.Alpha
  ) {materialFunction=&MF_Figment_BlendMatSpec}
  n39 = MaterialFunctionCall(
   In=@n38.Result
   Color=StaticSwitchParameter(True=LinearInterpolate(A=@n20, B=@n21, Alpha=@n22), False=@n20) {group=Utilities, parameterName=EnableQuickColorRender}
  ) {materialFunction=&MF_Athena_Material_BaseMultiply}
  n40 = ComponentMask(None=WorldPosition().XYZ) {g=true, r=true}
  n41 = Constant() {r=306000}
  n42 = Constant() {r=-153000}
  n43 = TextureSampleParameter2D(UVs=Subtract(A=Divide(A=@n40, B=@n41), B=Divide(A=@n42, B=@n41))) {
   parameterName=BiomeColormap
   samplerType=SAMPLERTYPE_LinearColor
   texture=&T_Figment_Terrain_CombinedColors_S9
  }
  n44 = Multiply(
   A=@n43.RGB
   B=CollectionParameter() {collection=&MPC_Landscape, parameterName=ColormapDivisor}
  )
  n45 = Multiply(A=GetMaterialAttributes(None=@n38.Result) {attributeGetTypes [#1]}.Base Color, B=@n44)
  n46 = MaterialFunctionCall!(
   Grass=NamedRerouteUsage()
   Material_Standard=@n39.Result
   Material_BaseColor=@n45
   Mask=LandscapeLayerSample() {parameterName=Arid_Grass_Light}
  ) {materialFunction=/FigmentEnvironment/LM_8_1_1/Athena/Environments/Landscape/MaterialFunctions/Arid/MF_Athena_AridGrass}
  n47 = MaterialFunctionCall!(
   BaseColor=LandscapeLayerSwitch!(
    LayerUsed=GetMaterialAttributes(None=@n46.Material_Standard) {attributeGetTypes [#1]}.Base Color
    LayerNotUsed=@n45
   ) {desc=Low/Medium, parameterName=Arid_Grass_Light}
   Mask=LandscapeLayerSample() {parameterName=Jungle_Pools, previewWeight=1}
  ) {materialFunction=/FigmentEnvironment/LM_8_1_1/Athena/Environments/Landscape/MaterialFunctions/Jungle/MF_Athena_JunglePools}
  n48 = MaterialFunctionCall!(UVs=LandscapeLayerCoords() {mappingScale=4}) {materialFunction=/FigmentEnvironment/LM_8_1_1/Athena/Environments/Landscape/MaterialFunctions/Standard/MF_Athena_Forest_01}
  n49 = LinearInterpolate(
   A=LinearInterpolate(A=@n47.Result, B=@n16.BaseColor_NoPebbles, Alpha=@n14.Lerp Alpha No Contrast)
   B=@n48.BaseColor
   Alpha=MaterialFunctionCall(In=LandscapeLayerSample() {parameterName=Forest, previewWeight=1}, Contrast=Constant() {r=1}) {materialFunction=&CheapContrast}.Result
  )
  n50 = ComponentMask(None=CollectionParameter() {collection=&MPC_Landscape, parameterName=Arid_Mud_Tint}) {b=true, g=true, r=true}
  n51 = MaterialFunctionCall() {materialFunction=&MF_Athena_RoadGravel}
  n52 = LandscapeLayerSwitch(
   LayerUsed=LinearInterpolate(A=@n51.Mud, B=Multiply(A=@n51.Mud, B=@n50), Alpha=LandscapeLayerSample() {parameterName=Arid})
   LayerNotUsed=@n51.Mud
  ) {parameterName=Arid}
  n53 = Constant3Vector() {constant=<rgba 3 1.6777 1.132 1>}
  n54 = LandscapeLayerSample() {parameterName=Farm_Grass}
  n55 = LandscapeLayerSwitch(LayerUsed=LinearInterpolate(A=@n52, B=Multiply(A=@n52, B=@n53), Alpha=@n54), LayerNotUsed=@n52) {parameterName=Farm_Grass}
  n56 = MaterialFunctionCall(In=LandscapeLayerSample() {parameterName=Mud}, Contrast=Constant() {r=1}) {materialFunction=&CheapContrast}
  n57 = LandscapeLayerSample() {parameterName=Road, previewWeight=1}
  n58 = OneMinus(None=NamedRerouteUsage())
  n59 = NamedRerouteUsage()
  n60 = ComponentMask(None=CollectionParameter() {collection=&MPC_Landscape, parameterName=Arid_Gravel_Tint}) {b=true, g=true, r=true}
  n61 = LandscapeLayerSwitch(
   LayerUsed=LinearInterpolate(A=@n51.Gravel, B=Multiply(A=@n51.Gravel_NoColor, B=@n60), Alpha=LandscapeLayerSample() {parameterName=Arid})
   LayerNotUsed=@n51.Gravel
  ) {parameterName=Arid}
  n62 = CollectionParameter() {collection=&MPC_Landscape, parameterName=Winter_Gravel_Tint}
  n63 = LandscapeLayerSample() {parameterName=Snow}
  n64 = LandscapeLayerSample() {parameterName=Ice}
  n65 = LandscapeLayerSample() {parameterName=SnowClumps}
  n66 = Clamp(None=Add(A=Add(A=@n63, B=@n64), B=@n65))
  n67 = LinearInterpolate(A=@n61, B=Multiply(A=@n51.Gravel_NoColor, B=ComponentMask(None=@n62) {b=true, g=true, r=true}), Alpha=@n66)
  n68 = LandscapeLayerSwitch(LayerUsed=@n67, LayerNotUsed=LandscapeLayerSwitch(LayerUsed=@n67, LayerNotUsed=@n61) {parameterName=Snow}) {parameterName=Ice}
  n69 = Constant3Vector() {constant=<rgba 0.9844 0.7659 0.6562 0>}
  n70 = LinearInterpolate(
   A=LinearInterpolate(
    A=LinearInterpolate(A=@n49, B=@n55, Alpha=@n56.Result)
    B=@n51.Road
    Alpha=MaterialFunctionCall(Transition Phase=@n57, Height Texture=@n58) {materialFunction=&HeightLerp}.Lerp Alpha No Contrast
   )
   B=LinearInterpolate(A=@n68, B=Multiply(A=@n68, B=@n69), Alpha=LandscapeLayerSample() {parameterName=Jungle_Pools})
   Alpha=@n59
  )
  n71 = Saturate(None=CollectionParameter() {collection=&MPC_Landscape, parameterName=DisplayLowDetailShader})
  n72 = GetMaterialAttributes(None=LandscapeLayerSwitch(LayerUsed=@n46.Material_Standard, LayerNotUsed=@n39.Result) {parameterName=Arid_Grass_Light}) {attributeGetTypes [#1]}
  n73 = LandscapeLayerSample() {parameterName=Jungle_Pools, previewWeight=1}
  n74 = GetMaterialAttributes(None=SetMaterialAttributes!(
    MaterialAttributes=@n72.MaterialAttributes
    Base Color=MaterialFunctionCall(BaseColor=@n72.Base Color, Mask=@n73) {materialFunction=&MF_Athena_JunglePools}.Result
   ) {attributeSetTypes [#1]}) {attributeGetTypes [
    #3
    #2
    #1
    #5
   ]}
  n75 = LinearInterpolate(
   A=@n74.Base Color
   B=@n47.Result
   Alpha=Saturate(None=CollectionParameter() {collection=&MPC_Landscape, parameterName=DisplayLowDetailShader})
  )
  n76 = Multiply(A=Desaturation(None=@n48.BaseColor, Fraction=Constant() {r=1}), B=Constant3Vector() {constant=<rgba 0.1303 1.5 0.2394 0>})
  n77 = MaterialFunctionCall(Transition Phase=Constant() {r=0.6}, Height Texture=@n48.Height, Contrast=Constant() {r=1}) {materialFunction=&HeightLerp}
  n78 = MaterialFunctionCall(Transition Phase=LandscapeLayerSample() {parameterName=Forest, previewWeight=1}, Height Texture=@n48.Height) {materialFunction=&HeightLerp}
  n79 = LinearInterpolate(
   A=LinearInterpolate(
    A=StaticSwitchParameter(True=@n75, False=@n74.Base Color) {group=Utilities, parameterName=EnableQuickColorRender}
    B=@n16.BaseColor_NoPebbles
    Alpha=@n14.Alpha
   )
   B=LinearInterpolate(
    A=@n48.BaseColor
    B=LinearInterpolate(A=@n48.BaseColor, B=@n76, Alpha=@n77.Alpha)
    Alpha=LandscapeLayerSample() {parameterName=Farm_Grass, previewWeight=1}
   )
   Alpha=@n78.Lerp Alpha No Contrast
  )
  n80 = Multiply(A=@n48.BaseColor, B=Constant3Vector() {constant=<rgba 1.5 1.2944 0.7891 1>})
  n81 = Multiply(A=@n52, B=Constant3Vector() {constant=<rgba 3 1.6777 1.132 1>})
  n82 = LandscapeLayerSample() {parameterName=Mud}
  n83 = MaterialFunctionCall(Transition Phase=LinearInterpolate(A=Constant() {r=-1}, B=Constant() {r=0.5}, Alpha=@n82), Height Texture=OneMinus(None=@n48.Height)) {materialFunction=&HeightLerp}
  n84 = LandscapeLayerSwitch(LayerUsed=LinearInterpolate(A=@n52, B=LinearInterpolate(A=@n80, B=@n81, Alpha=@n83.Lerp Alpha No Contrast), Alpha=@n54), LayerNotUsed=@n52) {parameterName=Farm_Grass}
  n85 = OneMinus(None=NamedRerouteUsage())
  n86 = LandscapeLayerSample() {parameterName=Arid}
  n87 = MaterialFunctionCall(
   Transition Phase=@n82
   Height Texture=LandscapeLayerSwitch(LayerUsed=LinearInterpolate(A=@n85, B=@n51.Gravel_Height, Alpha=@n86) {desc=Change heightmap used for blend in Arid}, LayerNotUsed=@n85) {parameterName=Arid}
   Contrast=Constant() {r=0.2}
  ) {materialFunction=&HeightLerp}
  n88 = MaterialFunctionCall(
   Transition Phase=Power(Base=LandscapeLayerSample() {parameterName=Road, previewWeight=1})
   Height Texture=@n51.Road_Height
   Contrast=Constant() {r=5}
  ) {materialFunction=&HeightLerp}
  n89 = Multiply(A=LandscapeLayerSample() {parameterName=Road, previewWeight=1}, B=Constant() {r=1.2})
  n90 = Constant() {r=0.25}
  n91 = MaterialFunctionCall(UseLandscapeCoordsOnSwitch=StaticBool() {value=true}) {materialFunction=&MF_Athena_Snow_01}
  n92 = MaterialFunctionCall(
   Transition Phase=Clamp(None=Subtract(A=@n89, B=@n65))
   Height Texture=LinearInterpolate(A=Multiply(A=@n58, B=@n90), B=@n91.Height, Alpha=@n66)
   Contrast=Constant() {r=2}
  ) {materialFunction=&HeightLerp}
  n93 = Multiply(A=@n68, B=Constant3Vector() {constant=<rgba 1.5 1.1733 0.8777 0>})
  n94 = LandscapeLayerSample() {parameterName=Gravel, previewWeight=1}
  n95 = Add(
   A=Multiply(A=@n94, B=ScalarParameter() {defaultValue=1.5, parameterName=GravelPool-X})
   B=ScalarParameter() {defaultValue=-1, parameterName=GravelPool-+}
  )
  n96 = LinearInterpolate(
   A=Constant3Vector() {constant=<rgba 0.12 0.0641 0.036 0>}
   B=@n93
   Alpha=MaterialFunctionCall(Transition Phase=Clamp(None=@n95), Height Texture=@n51.Gravel_Height) {materialFunction=&HeightLerp}.Lerp Alpha No Contrast
  )
  n97 = Clamp(None=LinearInterpolate(A=Constant() {r=-20}, B=Constant() {r=1}, Alpha=@n94))
  n98 = LandscapeLayerSample() {parameterName=Jungle_Pools}
  n99 = LandscapeLayerSwitch(LayerUsed=LinearInterpolate(A=@n68, B=LinearInterpolate(A=@n93, B=@n96, Alpha=@n97), Alpha=Ceil(None=@n98)), LayerNotUsed=@n68) {parameterName=Jungle_Pools}
  n100 = MaterialFunctionCall(
   Transition Phase=@n94
   Height Texture=LinearInterpolate(
    A=@n51.Gravel_Height
    B=Constant()
    Alpha=MaterialFunctionCall!(Distance) {materialFunction=/FigmentEnvironment/LM_8_1_1/Athena/Environments/Landscape/MaterialFunctions/MF_TerrainDistanceFade}.Result
   ) {constB=0.5}
  ) {materialFunction=&HeightLerp}
  n101 = LinearInterpolate(
   A=LinearInterpolate(A=LinearInterpolate(A=@n79, B=@n84, Alpha=@n87.Alpha), B=LinearInterpolate(A=@n84, B=@n51.Road, Alpha=@n88.Alpha), Alpha=@n92.Alpha)
   B=@n99
   Alpha=@n100.Lerp Alpha No Contrast
  )
  n102 = LinearInterpolate(
   A=StaticSwitchParameter(True=LinearInterpolate(A=@n101, B=@n70, Alpha=@n71), False=@n101) {group=Utilities, parameterName=EnableQuickColorRender}
   B=@n101
   Alpha=LandscapeLayerSample() {parameterName=NoTopo}
  )
  MaterialExpressionConstant_4 = Constant() {r=1}
  MaterialExpressionLinearInterpolate_41 = LinearInterpolate(
   A=ComponentMask(None=CollectionParameter() {collection=&MPC_Landscape, parameterName=Arid_Rock_01_Tint}) {b=true, g=true, r=true}
   B=Constant3Vector() {constant=<rgba 1 0.7455 0.5781 0>}
   Alpha=CollectionParameter() {collection=&MPC_Landscape, parameterName=Display_Combined}
  )
  n103 = LandscapeLayerBlend!(
   Layer Grass_Topology=@MaterialExpressionConstant_4
   Layer Arid=@MaterialExpressionLinearInterpolate_41
   Layer Ice=@MaterialExpressionConstant_4
   Layer Snow=@MaterialExpressionConstant_4
   Layer Grass_Only=@MaterialExpressionConstant_4
  ) {layers [
    {
     blendType=LB_WeightBlend
     constHeightInput=0
     layerName=Grass_Topology
     previewWeight=1
     constLayerInput=<0 0 0>
     heightInput {expression=null, inputName=None, mask=0, maskA=0, maskB=0, maskG=0, maskR=0, outputIndex=0}
     layerInput {
      expression=&self:MaterialExpressionConstant_4
      inputName=None
      mask=0
      maskA=0
      maskB=0
      maskG=0
      maskR=0
      outputIndex=0
     }
    }
    {
     blendType=LB_WeightBlend
     constHeightInput=0
     layerName=Arid
     previewWeight=1
     constLayerInput=<0 0 0>
     heightInput {expression=null, inputName=None, mask=0, maskA=0, maskB=0, maskG=0, maskR=0, outputIndex=0}
     layerInput {
      expression=&self:MaterialExpressionLinearInterpolate_41
      inputName=None
      mask=0
      maskA=0
      maskB=0
      maskG=0
      maskR=0
      outputIndex=0
     }
    }
    {
     blendType=LB_WeightBlend
     constHeightInput=0
     layerName=Ice
     previewWeight=0
     constLayerInput=<0 0 0>
     heightInput {expression=null, inputName=None, mask=0, maskA=0, maskB=0, maskG=0, maskR=0, outputIndex=0}
     layerInput {
      expression=&self:MaterialExpressionConstant_4
      inputName=None
      mask=0
      maskA=0
      maskB=0
      maskG=0
      maskR=0
      outputIndex=0
     }
    }
    {
     blendType=LB_WeightBlend
     constHeightInput=0
     layerName=Snow
     previewWeight=0
     constLayerInput=<0 0 0>
     heightInput {expression=null, inputName=None, mask=0, maskA=0, maskB=0, maskG=0, maskR=0, outputIndex=0}
     layerInput {
      expression=&self:MaterialExpressionConstant_4
      inputName=None
      mask=0
      maskA=0
      maskB=0
      maskG=0
      maskR=0
      outputIndex=0
     }
    }
    {
     blendType=LB_WeightBlend
     constHeightInput=0
     layerName=Grass_Only
     previewWeight=0
     constLayerInput=<0 0 0>
     heightInput {expression=null, inputName=None, mask=0, maskA=0, maskB=0, maskG=0, maskR=0, outputIndex=0}
     layerInput {
      expression=&self:MaterialExpressionConstant_4
      inputName=None
      mask=0
      maskA=0
      maskB=0
      maskG=0
      maskR=0
      outputIndex=0
     }
    }
   ]}
  n104 = CollectionParameter() {collection=&MPC_Landscape, parameterName=MountainRockTint}
  n105 = ComponentMask(None=CollectionParameter() {collection=&MPC_Landscape, parameterName=FarmRockTint}) {b=true, g=true, r=true}
  n106 = ScalarParameter() {parameterName=JungleRockLowDesat}
  n107 = LinearInterpolate(
   A=LinearInterpolate(A=@n103, B=ComponentMask(None=@n104) {b=true, g=true, r=true}, Alpha=LandscapeLayerSample() {parameterName=Mountain_Grass})
   B=LinearInterpolate(
    A=@n105
    B=Desaturation(None=@n105, Fraction=@n106)
    Alpha=CollectionParameter() {collection=&MPC_Landscape, parameterName=Display_Combined}
   )
   Alpha=LandscapeLayerSample() {parameterName=Farm_Grass}
  )
  MaterialExpressionConstant_3 = Constant() {r=1}
  MaterialExpressionMultiply_2 = Multiply(
   A=MaterialFunctionCall() {materialFunction=/FigmentEnvironment/LM_8_1_1/Athena/Environments/Landscape/MaterialFunctions/MF_Athena_ZSandColor}.Result
   B=ComponentMask(None=CollectionParameter() {collection=&MPC_Landscape, parameterName=Arid_Sand_Tint}) {b=true, g=true, r=true}
  )
  n108 = LandscapeLayerBlend(
   Layer Grass_Topology=@MaterialExpressionConstant_3
   Layer Arid=@MaterialExpressionMultiply_2
   Layer Ice=@MaterialExpressionConstant_3
   Layer Snow=@MaterialExpressionConstant_3
   Layer Grass_Only=@MaterialExpressionConstant_3
  ) {layers [
    {
     blendType=LB_WeightBlend
     constHeightInput=0
     layerName=Grass_Topology
     previewWeight=1
     constLayerInput=<0 0 0>
     heightInput {expression=null, inputName=None, mask=0, maskA=0, maskB=0, maskG=0, maskR=0, outputIndex=0}
     layerInput {
      expression=&self:MaterialExpressionConstant_3
      inputName=None
      mask=0
      maskA=0
      maskB=0
      maskG=0
      maskR=0
      outputIndex=0
     }
    }
    {
     blendType=LB_WeightBlend
     constHeightInput=0
     layerName=Arid
     previewWeight=1
     constLayerInput=<0 0 0>
     heightInput {expression=null, inputName=None, mask=0, maskA=0, maskB=0, maskG=0, maskR=0, outputIndex=0}
     layerInput {
      expression=&self:MaterialExpressionMultiply_2
      inputName=None
      mask=0
      maskA=0
      maskB=0
      maskG=0
      maskR=0
      outputIndex=0
     }
    }
    {
     blendType=LB_WeightBlend
     constHeightInput=0
     layerName=Ice
     previewWeight=0
     constLayerInput=<0 0 0>
     heightInput {expression=null, inputName=None, mask=0, maskA=0, maskB=0, maskG=0, maskR=0, outputIndex=0}
     layerInput {
      expression=&self:MaterialExpressionConstant_3
      inputName=None
      mask=0
      maskA=0
      maskB=0
      maskG=0
      maskR=0
      outputIndex=0
     }
    }
    {
     blendType=LB_WeightBlend
     constHeightInput=0
     layerName=Snow
     previewWeight=0
     constLayerInput=<0 0 0>
     heightInput {expression=null, inputName=None, mask=0, maskA=0, maskB=0, maskG=0, maskR=0, outputIndex=0}
     layerInput {
      expression=&self:MaterialExpressionConstant_3
      inputName=None
      mask=0
      maskA=0
      maskB=0
      maskG=0
      maskR=0
      outputIndex=0
     }
    }
    {
     blendType=LB_WeightBlend
     constHeightInput=0
     layerName=Grass_Only
     previewWeight=0
     constLayerInput=<0 0 0>
     heightInput {expression=null, inputName=None, mask=0, maskA=0, maskB=0, maskG=0, maskR=0, outputIndex=0}
     layerInput {
      expression=&self:MaterialExpressionConstant_3
      inputName=None
      mask=0
      maskA=0
      maskB=0
      maskG=0
      maskR=0
      outputIndex=0
     }
    }
   ]}
  n109 = CollectionParameter() {collection=&MPC_Landscape, parameterName=MountainGrassTint}
  n110 = Subtract(A=LandscapeLayerSample() {parameterName=Mountain_Grass}, B=LandscapeLayerSample() {parameterName=Snow})
  n111 = LinearInterpolate(
   A=LinearInterpolate(A=@n108, B=ComponentMask(None=@n109) {b=true, g=true, r=true}, Alpha=Clamp(None=@n110))
   B=ComponentMask(None=CollectionParameter() {collection=&MPC_Landscape, parameterName=FarmGrassTint}) {b=true, g=true, r=true}
   Alpha=LandscapeLayerSample() {parameterName=Farm_Grass}
  )
  n112 = LinearInterpolate(
   A=LinearInterpolate(
    A=@n107
    B=@n111
    Alpha=CollectionParameter() {collection=&MPC_Landscape, parameterName=Display_GroundColors}
   )
   B=LinearInterpolate(A=@n107, B=@n111, Alpha=NamedRerouteUsage())
   Alpha=CollectionParameter() {collection=&MPC_Landscape, parameterName=Display_Combined}
  )
  n113 = Multiply(A=@n112) {constB=2}
  n114 = TextureSample(UVs=MaterialFunctionCall() {materialFunction=&MF_Athena_Terrain_Coords}.Result) {
   samplerSource=SSM_Wrap_WorldGroupSettings
   samplerType=SAMPLERTYPE_LinearColor
   texture=/FigmentEnvironment/LM_8_1_1/Athena/Environments/Landscape/Textures/T_Athena_Terrain_Topo_Mask
  }
  n115 = CollectionParameter() {collection=&MPC_Landscape, parameterName=Display_Combined}
  n116 = LinearInterpolate(A=Constant() {r=1}, B=ScalarParameter() {defaultValue=2, parameterName=RockEdgeDesat}, Alpha=@n115)
  n117 = ScalarParameter() {parameterName=RockEdgeDesat_Arid}
  n118 = LandscapeLayerSample() {parameterName=Arid}
  n119 = OneMinus(None=MaterialFunctionCall(Height=MaterialFunctionCall() {materialFunction=&MF_Athena_Grass_01}.Height) {materialFunction=&MF_Athena_SlopeBlend}.Alpha_NoContrast)
  n120 = Abs(None=ComponentMask(None=VertexNormalWS()) {b=true})
  n121 = MaterialFunctionCall() {materialFunction=&MF_Athena_Rock_01}
  n122 = Min(
   A=@n119
   B=ComponentMask(None=MaterialFunctionCall(In=@n120, Contrast=@n121.Proj_Contrast) {materialFunction=&CheapContrast}.Result) {r=true}
  )
  n123 = LinearInterpolate(
   B=LinearInterpolate(
    A=LinearInterpolate(A=@n116, B=@n117, Alpha=@n118)
    B=ScalarParameter() {parameterName=RockEdgeDesat_Jungle}
    Alpha=LandscapeLayerSample() {parameterName=Farm_Grass}
   )
   Alpha=Clamp(None=LinearInterpolate(A=@n119, B=@n122, Alpha=@n115))
  )
  n124 = Desaturation(
   None=LinearInterpolate(A=Multiply(A=@n113, B=@n114.G), B=@n112, Alpha=LandscapeLayerSample() {parameterName=NoTopo})
   Fraction=LinearInterpolate(
    B=Clamp(None=@n123)
    Alpha=CollectionParameter() {collection=&MPC_Landscape, parameterName=Display_Combined}
   ) {constB=0}
  ) {desc=To Render Colormap}
  n125 = OneMinus(None=@n114.G)
  n126 = Subtract(A=Clamp(None=LinearInterpolate(Alpha=@n125) {constA=-2, constB=2}), B=LandscapeLayerSample() {parameterName=NoTopo})
  n127 = NamedRerouteUsage()
  n128 = LinearInterpolate(A=@n124, B=Multiply(A=@n124, B=Constant3Vector() {constant=<rgba 0.8536 0.9838 1.1 1>}), Alpha=Clamp(None=Multiply(A=@n126, B=@n127)))
  n129 = CollectionParameter() {collection=&MPC_Landscape, parameterName=ColormapDivisor}
  n130 = CollectionParameter() {collection=&MPC_Landscape, parameterName=GrassEdgeZ}
  n131 = ComponentMask(None=VertexNormalWS()) {b=true}
  n132 = CollectionParameter() {collection=&MPC_Landscape, parameterName=GrassEdgeDesatWidth}
  n133 = MaterialFunctionCall!(UVW=Desaturation(None=Divide(A=@n128, B=@n129), Fraction=SphereMask(A=@n130, B=@n131, Radius=@n132) {hardnessPercent=0})) {materialFunction=/Engine/ArtTools/RenderToTexture/MaterialFunctions/MarkRedOutside0-1Range}
  n134 = Saturate(None=CollectionParameter() {collection=&MPC_Landscape, parameterName=RenderColorMap})
  MaterialExpressionConstant_9 = Constant()
  MaterialExpressionConstant3Vector_1 = Constant3Vector() {constant=<rgba 0 0 1 0>}
  MaterialExpressionConstant3Vector_3 = Constant3Vector() {desc=Mountain, constant=<rgba 0 1 0 0>}
  n135 = LandscapeLayerBlend!(
   Layer Grass_Topology=@MaterialExpressionConstant_9
   Layer Arid=@MaterialExpressionConstant3Vector_1
   Layer Farm_Grass=@MaterialExpressionConstant_9
   Layer Mountain_Grass=@MaterialExpressionConstant_9
   Layer Ice=@MaterialExpressionConstant3Vector_3
   Layer Snow=@MaterialExpressionConstant3Vector_3
   Layer Grass_Only=@MaterialExpressionConstant_9
  ) {layers [
    {
     blendType=LB_WeightBlend
     constHeightInput=0
     layerName=Grass_Topology
     previewWeight=1
     constLayerInput=<0 0 0>
     heightInput {expression=null, inputName=None, mask=0, maskA=0, maskB=0, maskG=0, maskR=0, outputIndex=0}
     layerInput {
      expression=&self:MaterialExpressionConstant_9
      inputName=None
      mask=0
      maskA=0
      maskB=0
      maskG=0
      maskR=0
      outputIndex=0
     }
    }
    {
     blendType=LB_WeightBlend
     constHeightInput=0
     layerName=Arid
     previewWeight=1
     constLayerInput=<0 0 0>
     heightInput {expression=null, inputName=None, mask=0, maskA=0, maskB=0, maskG=0, maskR=0, outputIndex=0}
     layerInput {
      expression=&self:MaterialExpressionConstant3Vector_1
      inputName=None
      mask=0
      maskA=0
      maskB=0
      maskG=0
      maskR=0
      outputIndex=0
     }
    }
    {
     blendType=LB_WeightBlend
     constHeightInput=0
     layerName=Farm_Grass
     previewWeight=1
     constLayerInput=<0 0 0>
     heightInput {expression=null, inputName=None, mask=0, maskA=0, maskB=0, maskG=0, maskR=0, outputIndex=0}
     layerInput {
      expression=&self:MaterialExpressionConstant_9
      inputName=None
      mask=0
      maskA=0
      maskB=0
      maskG=0
      maskR=0
      outputIndex=0
     }
    }
    {
     blendType=LB_WeightBlend
     constHeightInput=0
     layerName=Mountain_Grass
     previewWeight=1
     constLayerInput=<0 0 0>
     heightInput {expression=null, inputName=None, mask=0, maskA=0, maskB=0, maskG=0, maskR=0, outputIndex=0}
     layerInput {
      expression=&self:MaterialExpressionConstant_9
      inputName=None
      mask=0
      maskA=0
      maskB=0
      maskG=0
      maskR=0
      outputIndex=0
     }
    }
    {
     blendType=LB_WeightBlend
     constHeightInput=0
     layerName=Ice
     previewWeight=0
     constLayerInput=<0 0 0>
     heightInput {expression=null, inputName=None, mask=0, maskA=0, maskB=0, maskG=0, maskR=0, outputIndex=0}
     layerInput {
      expression=&self:MaterialExpressionConstant3Vector_3
      inputName=None
      mask=0
      maskA=0
      maskB=0
      maskG=0
      maskR=0
      outputIndex=0
     }
    }
    {
     blendType=LB_WeightBlend
     constHeightInput=0
     layerName=Snow
     previewWeight=0
     constLayerInput=<0 0 0>
     heightInput {expression=null, inputName=None, mask=0, maskA=0, maskB=0, maskG=0, maskR=0, outputIndex=0}
     layerInput {
      expression=&self:MaterialExpressionConstant3Vector_3
      inputName=None
      mask=0
      maskA=0
      maskB=0
      maskG=0
      maskR=0
      outputIndex=0
     }
    }
    {
     blendType=LB_WeightBlend
     constHeightInput=0
     layerName=Grass_Only
     previewWeight=0
     constLayerInput=<0 0 0>
     heightInput {expression=null, inputName=None, mask=0, maskA=0, maskB=0, maskG=0, maskR=0, outputIndex=0}
     layerInput {
      expression=&self:MaterialExpressionConstant_9
      inputName=None
      mask=0
      maskA=0
      maskB=0
      maskG=0
      maskR=0
      outputIndex=0
     }
    }
   ]}
  n136 = Constant3Vector() {desc=Farm, constant=<rgba 1 0 0 0>}
  n137 = LandscapeLayerSample() {parameterName=Jungle_Pools}
  n138 = CollectionParameter() {collection=&MPC_Landscape, parameterName=DisplayBiomeRGB}
  n139 = StaticSwitchParameter(
   True=LinearInterpolate(A=LinearInterpolate(A=@n102, B=@n133.Result, Alpha=@n134), B=LinearInterpolate(A=@n135, B=@n136, Alpha=@n137), Alpha=Saturate(None=@n138))
   False=@n102
  ) {group=Utilities, parameterName=EnableQuickColorRender}
  n140 = LinearInterpolate(
   A=LinearInterpolate(A=@n47.Result, B=@n16.BaseColor_NoPebbles, Alpha=@n14.Lerp Alpha No Contrast)
   B=@n48.BaseColor
   Alpha=@n78.Lerp Alpha No Contrast
  )
  n141 = LinearInterpolate(
   A=LinearInterpolate(A=LinearInterpolate(A=@n140, B=@n84, Alpha=@n87.Alpha), B=@n51.Road, Alpha=@n92.Alpha)
   B=@n99
   Alpha=@n100.Lerp Alpha No Contrast
  )
  n142 = QualitySwitch(Default=@n139, Low=@n141, High=@n139, Medium=@n141)
  n143 = FeatureLevelSwitch(Default=@n142, ES2=@n70, ES3_1=@n70, SM4_REMOVED=@n142, SM5=@n142, SM6=@n142)
  n144 = GetMaterialAttributes(None=MaterialFunctionCall!(
    In=MakeMaterialAttributes!(
     BaseColor=@n143
     Metallic
     Specular
     Roughness
     Anisotropy
     EmissiveColor
     Opacity
     OpacityMask
     Normal
     Tangent
     WorldPositionOffset
     SubsurfaceColor
     ClearCoat
     ClearCoatRoughness
     AmbientOcclusion
     Refraction
     CustomizedUVs_0
     CustomizedUVs_1
     CustomizedUVs_2
     CustomizedUVs_3
     CustomizedUVs_4
     CustomizedUVs_5
     CustomizedUVs_6
     CustomizedUVs_7
     PixelDepthOffset
     ShadingModel
     Displacement
    )
    Rock_SlopeMask
    GrassMask
    BaseColorTintOnlybyVertColor
    VertexColorMask
   ) {materialFunction=&/MF_CattusSpiral}.Material Out) {attributeGetTypes [#1]}
  n145 = Multiply(A=ComponentMask(None=@n144.Base Color) {r=true}, B=Constant() {r=1})
  n146 = FeatureLevelSwitch(Default=NamedRerouteUsage(), ES2=@n145, ES3_1=@n145)
  n147 = Constant() {r=1}
  n148 = Clamp(None=Add(A=@n2, B=LandscapeLayerSample() {parameterName=Forest, previewWeight=1}))
  n149 = LinearInterpolate(A=@n74.Roughness, Alpha=@n148)
  n150 = QualitySwitch(Default=@n149, Low=@n147, High=@n149, Medium=@n147)
  n151 = FeatureLevelSwitch(Default=@n150, ES2=@n147, ES3_1=@n147, SM4_REMOVED=@n150, SM5=@n150, SM6=@n150)
  n152 = LinearInterpolate(A=@n74.Normal, B=Transform(None=@n74.Normal), Alpha=NamedRerouteUsage())
  n153 = ScalarParameter() {defaultValue=1, desc=Debug to view Normals without fix, parameterName=UseRVTNormalSpaceSwitch}
  n154 = LinearInterpolate(
   A=LinearInterpolate(
    A=@n74.Normal
    B=MaterialFunctionCall!(NoRVT=@n74.Normal, RVT=@n152) {materialFunction=/Game/Packages/Fortress_SharedMaterials/Base_Material_Functions/FN_RVT_Switch_Color}.Result
    Alpha=Saturate(None=@n153)
   )
   B=Constant3Vector() {constant=<rgba 0 0 1 0>}
   Alpha=Clamp(None=@n2)
  )
  n155 = GetMaterialAttributes(None=NamedRerouteUsage()) {attributeGetTypes [#5]}
  n156 = NamedRerouteUsage()
  n157 = WorldPosition() {worldPositionShaderOffset=WPT_ExcludeAllShaderOffsets}
  n158 = TextureSample(UVs=Divide(A=ComponentMask(None=@n157.XYZ) {g=true, r=true}, B=Constant() {r=25000})) {
   samplerSource=SSM_Wrap_WorldGroupSettings
   samplerType=SAMPLERTYPE_Normal
   texture=/Game/ContentCreationTools/TextureCreation/Textures/T_BPCreated_MacroNormal_01
  }
  n159 = Constant() {r=0.625}
  n160 = Constant() {r=0.25}
  n161 = Constant() {r=0.9}
  n162 = FeatureLevelSwitch(Default=@n160, ES2=@n161, ES3_1=@n161, SM4_REMOVED=@n160)
  n163 = Multiply(
   A=LinearInterpolate(
    A=Constant3Vector() {constant=<rgba 0 0 1 0>}
    B=@n158.RGB
    Alpha=MaterialFunctionCall() {materialFunction=&MF_TerrainDistanceFade}.Result
   )
   B=MaterialFunctionCall(X=@n159, Y=@n159, Z=QualitySwitch(Default=@n162, Low=@n161, Medium=@n161)) {materialFunction=&MakeFloat3}.Result
  )
  n164 = Add(A=Multiply(A=LinearInterpolate(A=@n154, B=@n155.Normal, Alpha=@n156), B=Constant3Vector() {constant=<rgba 1 1 1 0>}), B=@n163)
  n165 = MaterialFunctionCall() {materialFunction=&MF_Athena_Snow_01}
  n166 = NamedRerouteUsage()
  n167 = Add(A=LandscapeLayerSample() {parameterName=Gravel}, B=LandscapeLayerSample() {parameterName=Forest})
  n168 = LandscapeLayerSample() {parameterName=Mud}
  n169 = Subtract(
   A=Add(A=LandscapeLayerSample() {parameterName=Ice}, B=LandscapeLayerSample() {parameterName=Snow})
   B=Add(A=Add(A=@n167, B=@n168), B=LandscapeLayerSample() {parameterName=Road})
  ) {desc=Mask out paths and dirt from emmisive}
  n170 = LinearInterpolate(A=Add(A=@n165.Normal, B=@n163), B=@n165.Normal, Alpha=Clamp(None=Multiply(A=@n166, B=@n169)))
  n171 = QualitySwitch(Default=@n164, Low=@n170, High=@n164, Medium=@n170)
  n172 = FeatureLevelSwitch(Default=@n171, ES2=@n163, ES3_1=@n163, SM4_REMOVED=@n171, SM5=@n171, SM6=@n171)
  n173 = SetMaterialAttributes!(MaterialAttributes, Base Color=@n144.Base Color, Specular=@n146, Roughness=@n151, Opacity Mask, Normal=@n172, Emissive Color) {attributeSetTypes [
    #1
    #2
    #3
    #4
    #5
    #6
   ]}
  n174 = MaterialFunctionCall(NoRVT=@n173, RVT=@n13) {materialFunction=&FN_RVT_Switch_Material}
  BlendMaterialAttributes(
   A=BlendMaterialAttributes(A=@n174.Result, B=NamedRerouteUsage(), Alpha=LandscapeLayerSample() {parameterName=Ice})
   B=NamedRerouteUsage()
   Alpha=LandscapeLayerSample() {parameterName=Snow}
  )
  Constant()
  n175 = Add(
   A=LandscapeLayerSample() {parameterName=Grass_Topology, previewWeight=1}
   B=Round(None=LandscapeLayerSample() {parameterName=Grass_Only, previewWeight=1})
  )
  n176 = Abs(None=ComponentMask(None=VertexNormalWS()) {b=true})
  n177 = ScalarParameter() {defaultValue=0.975, parameterName=GrassNormalMaskGain}
  n178 = ScalarParameter() {defaultValue=0.2, parameterName=ShiftEdge}
  n179 = Subtract(A=Clamp(None=Multiply(A=@n176, B=@n177)), B=@n178)
  n180 = OneMinus(None=@n178)
  n181 = ScalarParameter() {defaultValue=5, parameterName=SlopeContrast}
  n182 = MaterialFunctionCall(
   In=MaterialFunctionCall(Transition Phase=Divide(A=@n179, B=@n180), Height Texture=ScalarParameter() {parameterName=HeightT}, Contrast=@n181) {materialFunction=&HeightLerp}.Alpha
   Contrast=@n181
  ) {materialFunction=&CheapContrast}
  n183 = OneMinus(None=Saturate(None=@n182.Result))
  n184 = Clamp(None=@n183)
  n185 = LandscapeLayerSample() {parameterName=Mud}
  n186 = Add(A=Add(A=LandscapeLayerSample() {parameterName=Gravel}, B=@n184), B=@n185)
  n187 = LandscapeLayerSample() {parameterName=Road, previewWeight=1}
  n188 = LandscapeLayerSample() {parameterName=NoFoliage}
  n189 = Add(A=Add(A=Add(A=@n186, B=@n187), B=@n188), B=LandscapeLayerSample() {parameterName=Forest})
  n190 = Round(None=LandscapeLayerSample() {parameterName=Jungle_Sand, previewWeight=1})
  n191 = Add(A=@n189, B=@n190)
  n192 = Subtract(A=@n175, B=@n191)
  MaterialExpressionMaterialFunctionCall_58 = MaterialFunctionCall(Rock_SlopeMask=NamedRerouteUsage(), GrassMask=@n192) {materialFunction=&MF_CattusSpiral}
  n193 = Add(A=WorldPosition().Z, B=ScalarParameter() {defaultValue=5000, parameterName=GrassZCull})
  MaterialExpressionMultiply_39 = Multiply(A=@MaterialExpressionMaterialFunctionCall_58.GrassMask, B=Saturate(None=Sign(None=@n193)))
  n194 = Multiply(A=LandscapeLayerSample() {parameterName=Forest}, B=@n1)
  n195 = LandscapeLayerSample() {parameterName=Mud}
  n196 = LandscapeLayerSample() {parameterName=NoFoliage, previewWeight=1}
  MaterialExpressionReroute_7 = Reroute(None=Subtract(A=Subtract(A=@n194, B=@n195), B=@n196)) {desc=Jungle}
  MaterialExpressionReroute_55 = Reroute(None=Subtract(A=LandscapeLayerSample() {parameterName=CraterOvergrowth}, B=@n196)) {desc=farm}
  n197 = Subtract(A=LandscapeLayerSample() {parameterName=Forest}, B=Add(A=Add(A=@n185, B=@n187), B=@n188))
  MaterialExpressionSubtract_26 = Subtract(A=Subtract(A=Subtract(A=@n197, B=@n1), B=NamedRerouteUsage()), B=@n196)
  n198 = Subtract(
   A=Subtract(A=LandscapeLayerSample() {parameterName=Gravel}, B=@n184)
   B=Multiply(B=LandscapeLayerSample() {parameterName=Crater}) {constA=10}
  )
  n199 = LandscapeLayerSample() {parameterName=NoFoliage}
  n200 = LandscapeLayerSample() {parameterName=Ice}
  n201 = LandscapeLayerSample() {parameterName=Snow}
  MaterialExpressionReroute_50 = Reroute(None=Subtract(A=Subtract(A=@n198, B=@n199), B=Add(A=@n200, B=@n201))) {desc=Gravel}
  n202 = LandscapeLayerSample() {parameterName=Flowers, previewWeight=1}
  n203 = Add(A=LandscapeLayerSample() {parameterName=Snow}, B=LandscapeLayerSample() {parameterName=Ice})
  MaterialExpressionReroute_51 = Reroute(None=Subtract(A=Subtract(A=@n202, B=@n196), B=@n203)) {desc=Flowers}
  MaterialExpressionConstant_14 = Constant()
  n204 = Round(None=LandscapeLayerSample() {parameterName=Arid})
  n205 = Subtract(A=@n204, B=@n184) {desc=Subtract rock}
  n206 = Multiply(A=OneMinus(None=LandscapeLayerSample() {parameterName=Arid_Grass_light})) {constB=0.7, desc="Higher Value = Less Grass in non gras areaas"}
  MaterialExpressionReroute_53 = Reroute(None=Subtract(A=Subtract(A=@n205, B=@n206), B=@n191)) {desc=Sand}
  MaterialExpressionReroute_54 = Reroute(None=Multiply(A=Subtract(A=@n203, B=@n191), B=@n202)) {desc=snow}
  LandscapeGrassOutput!(
   RedFlowers=@MaterialExpressionMaterialFunctionCall_58.FlowerMask
   Grass=@MaterialExpressionMultiply_39
   Jungle=@MaterialExpressionReroute_7
   Farm=@MaterialExpressionReroute_55
   Forest=@MaterialExpressionSubtract_26
   Gravel=@MaterialExpressionReroute_50
   Flowers=@MaterialExpressionReroute_51
   JungleGlobal=@MaterialExpressionConstant_14
   Arid_Sand=@MaterialExpressionReroute_53
   Snow=@MaterialExpressionReroute_54
  ) {grassTypes [
    RedFlowers {
     grassType=/FigmentEnvironment/S10/Landscape/Foliage/Athena_Foliage_Flowers_Cattus
     input {
      expression=&self:MaterialExpressionMaterialFunctionCall_58
      inputName=None
      mask=0
      maskA=0
      maskB=0
      maskG=0
      maskR=0
      outputIndex=3
     }
    }
    Grass {
     grassType=/FigmentEnvironment/LM_8_1_1/Athena/Environments/Landscape/Foliage/Athena_Foliage_Grass
     input {
      expression=&self:MaterialExpressionMultiply_39
      inputName=None
      mask=0
      maskA=0
      maskB=0
      maskG=0
      maskR=0
      outputIndex=0
     }
    }
    Jungle {
     grassType=/FigmentEnvironment/LM_8_1_1/Athena/Environments/Landscape/Foliage/Athena_Foliage_Jungle
     input {
      expression=&self:MaterialExpressionReroute_7
      inputName=None
      mask=0
      maskA=0
      maskB=0
      maskG=0
      maskR=0
      outputIndex=0
     }
    }
    Farm {
     grassType=/FigmentEnvironment/LM_8_1_1/Athena/Environments/Landscape/Foliage/Athena_Foliage_CraterOvergrowth
     input {
      expression=&self:MaterialExpressionReroute_55
      inputName=None
      mask=0
      maskA=0
      maskB=0
      maskG=0
      maskR=0
      outputIndex=0
     }
    }
    Forest {
     grassType=/FigmentEnvironment/LM_8_1_1/Athena/Environments/Landscape/Foliage/Athena_Foliage_Forest
     input {
      expression=&self:MaterialExpressionSubtract_26
      inputName=None
      mask=0
      maskA=0
      maskB=0
      maskG=0
      maskR=0
      outputIndex=0
     }
    }
    Gravel {
     grassType=/FigmentEnvironment/LM_8_1_1/Athena/Environments/Landscape/Foliage/Athena_Foliage_Gravel
     input {
      expression=&self:MaterialExpressionReroute_50
      inputName=None
      mask=0
      maskA=0
      maskB=0
      maskG=0
      maskR=0
      outputIndex=0
     }
    }
    Flowers {
     grassType=/FigmentEnvironment/LM_8_1_1/Athena/Environments/Landscape/Foliage/Athena_Foliage_Flowers
     input {
      expression=&self:MaterialExpressionReroute_51
      inputName=None
      mask=0
      maskA=0
      maskB=0
      maskG=0
      maskR=0
      outputIndex=0
     }
    }
    JungleGlobal {
     grassType=/FigmentEnvironment/LM_8_1_1/Athena/Environments/Landscape/Foliage/Athena_Foliage_JungleGlobal
     input {
      expression=&self:MaterialExpressionConstant_14
      inputName=None
      mask=0
      maskA=0
      maskB=0
      maskG=0
      maskR=0
      outputIndex=0
     }
    }
    Arid_Sand {
     grassType=/FigmentEnvironment/LM_8_1_1/Athena/Environments/Landscape/Foliage/Athena_Foliage_Arid_Grass
     input {
      expression=&self:MaterialExpressionReroute_53
      inputName=None
      mask=0
      maskA=0
      maskB=0
      maskG=0
      maskR=0
      outputIndex=0
     }
    }
    Snow {
     grassType=/FigmentEnvironment/LM_8_1_1/Athena/Environments/Landscape/Foliage/Athena_Foliage_Snow
     input {
      expression=&self:MaterialExpressionReroute_54
      inputName=None
      mask=0
      maskA=0
      maskB=0
      maskG=0
      maskR=0
      outputIndex=0
     }
    }
   ]}
  n207 = MaterialFunctionCall!(AridMasks=@n15, ProjContrast=Constant() {r=10}) {materialFunction=/FigmentEnvironment/LM_8_1_1/Athena/Environments/Landscape/MaterialFunctions/Arid/MF_Athena_Arid_Rock}
  n208 = MaterialFunctionCall!(Material=@n207.Mat, NewBaseColor=@n207.Mat, Mask=LandscapeLayerSample() {parameterName=Arid_RockDetail, previewWeight=1}) {materialFunction=/Engine/Functions/MaterialLayerFunctions/MatLayerBlend_OverrideBaseColor}
  n209 = MaterialFunctionCall(
   In=@n208.Blended Material
   Color=MaterialFunctionCall() {materialFunction=/FigmentEnvironment/LM_8_1_1/Athena/Environments/Landscape/MaterialFunctions/MF_Athena_SedimentGradient}.Result
  ) {materialFunction=&MF_Athena_Material_BaseMultiply}
  n210 = MaterialFunctionCall(In=@n207.MatSimple, Color=Constant3Vector() {desc="Ship hack, Replace Later", constant=<rgba 1 0.7974 0.7166 0>}) {materialFunction=&MF_Athena_Material_BaseMultiply}
  n211 = QualitySwitch(Default=@n209.Result, Low=@n210.Result, High=@n209.Result, Medium=@n210.Result)
  n212 = SetMaterialAttributes!(MaterialAttributes=@n16.Simple_Material, Normal=@n91.Normal) {attributeSetTypes [#5]}
  n213 = MaterialFunctionCall(Height=@n91.Height, Transition Multiplier=@n31) {materialFunction=&MF_Athena_SlopeBlend}
  MaterialExpressionStaticSwitchParameter_7 = StaticSwitchParameter(
   True=MaterialFunctionCall(
    Base Material=FeatureLevelSwitch(Default=@n211, ES2=@n210.Result, ES3_1=@n210.Result, SM4_REMOVED=@n211, SM5=@n211, SM6=@n211)
    Top Material=@n212
    Alpha=@n213.Alpha
   ) {materialFunction=&MatLayerBlend_Standard}.Blended Material
   False=@n212
  ) {defaultValue=true, parameterName=UseAridRockBlend}
  MaterialExpressionNamedRerouteUsage_58 = NamedRerouteUsage()
  MaterialExpressionNamedRerouteUsage_57 = NamedRerouteUsage()
  MaterialExpressionNamedRerouteUsage_59 = NamedRerouteUsage()
  MaterialExpressionNamedRerouteUsage_49 = NamedRerouteUsage()
  MaterialExpressionNamedRerouteUsage_50 = NamedRerouteUsage()
  LandscapeLayerBlend!(
   Layer Grass_Topology=@MaterialExpressionMaterialFunctionCall_31.Blended Material
   Layer Arid=@MaterialExpressionStaticSwitchParameter_7
   Layer Ice=@MaterialExpressionNamedRerouteUsage_58
   Layer Snow=@MaterialExpressionNamedRerouteUsage_57
   Height Snow=@MaterialExpressionNamedRerouteUsage_59
   Layer Grass_Only=@MaterialExpressionNamedRerouteUsage_49
   Height Grass_Only=@MaterialExpressionNamedRerouteUsage_50
  ) {layers [
    {
     blendType=LB_WeightBlend
     constHeightInput=0
     layerName=Grass_Topology
     previewWeight=1
     constLayerInput=<0 0 0>
     heightInput {expression=null, inputName=None, mask=0, maskA=0, maskB=0, maskG=0, maskR=0, outputIndex=0}
     layerInput {
      expression=&self:MaterialExpressionMaterialFunctionCall_31
      inputName=None
      mask=0
      maskA=0
      maskB=0
      maskG=0
      maskR=0
      outputIndex=0
     }
    }
    {
     blendType=LB_WeightBlend
     constHeightInput=0
     layerName=Arid
     previewWeight=1
     constLayerInput=<0 0 0>
     heightInput {expression=null, inputName=None, mask=0, maskA=0, maskB=0, maskG=0, maskR=0, outputIndex=0}
     layerInput {
      expression=&self:MaterialExpressionStaticSwitchParameter_7
      inputName=None
      mask=0
      maskA=0
      maskB=0
      maskG=0
      maskR=0
      outputIndex=0
     }
    }
    {
     blendType=LB_WeightBlend
     constHeightInput=0
     layerName=Ice
     previewWeight=0
     constLayerInput=<0 0 0>
     heightInput {expression=null, inputName=None, mask=0, maskA=0, maskB=0, maskG=0, maskR=0, outputIndex=0}
     layerInput {
      expression=&self:MaterialExpressionNamedRerouteUsage_58
      inputName=None
      mask=0
      maskA=0
      maskB=0
      maskG=0
      maskR=0
      outputIndex=0
     }
    }
    {
     blendType=LB_HeightBlend
     constHeightInput=0
     layerName=Snow
     previewWeight=0
     constLayerInput=<0 0 0>
     heightInput {
      expression=&self:MaterialExpressionNamedRerouteUsage_59
      inputName=None
      mask=0
      maskA=0
      maskB=0
      maskG=0
      maskR=0
      outputIndex=0
     }
     layerInput {
      expression=&self:MaterialExpressionNamedRerouteUsage_57
      inputName=None
      mask=0
      maskA=0
      maskB=0
      maskG=0
      maskR=0
      outputIndex=0
     }
    }
    {
     blendType=LB_HeightBlend
     constHeightInput=0
     layerName=Grass_Only
     previewWeight=1
     constLayerInput=<0 0 0>
     heightInput {
      expression=&self:MaterialExpressionNamedRerouteUsage_50
      inputName=None
      mask=0
      maskA=0
      maskB=0
      maskG=0
      maskR=0
      outputIndex=0
     }
     layerInput {
      expression=&self:MaterialExpressionNamedRerouteUsage_49
      inputName=None
      mask=0
      maskA=0
      maskB=0
      maskG=0
      maskR=0
      outputIndex=0
     }
    }
   ]}
  LandscapeLayerSample() {parameterName=Flowers, previewWeight=1}
  MaterialExpressionNamedRerouteUsage_29 = NamedRerouteUsage()
  MaterialExpressionNamedRerouteUsage_25 = NamedRerouteUsage()
  MaterialExpressionNamedRerouteUsage_42 = NamedRerouteUsage()
  MaterialExpressionNamedRerouteUsage_23 = NamedRerouteUsage()
  MaterialExpressionNamedRerouteUsage_33 = NamedRerouteUsage()
  MaterialExpressionNamedRerouteUsage_21 = NamedRerouteUsage()
  LandscapePhysicalMaterialOutput!(
   Stone=@MaterialExpressionNamedRerouteUsage_29
   Asphalt=@MaterialExpressionNamedRerouteUsage_25
   Snow=@MaterialExpressionNamedRerouteUsage_42
   Dirt=@MaterialExpressionNamedRerouteUsage_23
   Sand=@MaterialExpressionNamedRerouteUsage_33
   Grass=@MaterialExpressionNamedRerouteUsage_21
  ) {inputs [
    {
     physicalMaterial=/Game/FortressPhysicalMaterials/Stone
     input {
      expression=&self:MaterialExpressionNamedRerouteUsage_29
      inputName=None
      mask=0
      maskA=0
      maskB=0
      maskG=0
      maskR=0
      outputIndex=0
     }
    }
    {
     physicalMaterial=/Game/FortressPhysicalMaterials/Asphalt
     input {
      expression=&self:MaterialExpressionNamedRerouteUsage_25
      inputName=None
      mask=0
      maskA=0
      maskB=0
      maskG=0
      maskR=0
      outputIndex=0
     }
    }
    {
     physicalMaterial=/Game/FortressPhysicalMaterials/Snow
     input {
      expression=&self:MaterialExpressionNamedRerouteUsage_42
      inputName=None
      mask=0
      maskA=0
      maskB=0
      maskG=0
      maskR=0
      outputIndex=0
     }
    }
    {
     physicalMaterial=/Game/FortressPhysicalMaterials/Dirt
     input {
      expression=&self:MaterialExpressionNamedRerouteUsage_23
      inputName=None
      mask=0
      maskA=0
      maskB=0
      maskG=0
      maskR=0
      outputIndex=0
     }
    }
    {
     physicalMaterial=/Game/FortressPhysicalMaterials/Sand
     input {
      expression=&self:MaterialExpressionNamedRerouteUsage_33
      inputName=None
      mask=0
      maskA=0
      maskB=0
      maskG=0
      maskR=0
      outputIndex=0
     }
    }
    {
     physicalMaterial=/Game/FortressPhysicalMaterials/Grass
     input {
      expression=&self:MaterialExpressionNamedRerouteUsage_21
      inputName=None
      mask=0
      maskA=0
      maskB=0
      maskG=0
      maskR=0
      outputIndex=0
     }
    }
   ]}
  LinearInterpolate(
   A=ScalarParameter() {defaultValue=-65, parameterName=RockAlphaBlendA}
   B=ScalarParameter() {defaultValue=50, parameterName=RockAlphaBlendB}
   Alpha=Add(A=@n183, B=@n178)
  )
  LinearInterpolate(
   A=LinearInterpolate(A=Constant() {r=0.1}, B=Constant() {r=0.25}, Alpha=LandscapeLayerSample() {parameterName=Ice})
   B=Clamp(None=GetMaterialAttributes(None=@n165.Material) {attributeGetTypes [#2]}.Specular)
   Alpha=LandscapeLayerSample() {parameterName=Snow}
  )
  n214 = LandscapeLayerSample() {parameterName=Gravel, previewWeight=1}
  LinearInterpolate(A=@n214, B=@n59, Alpha=ScalarParameter() {parameterName=UseMask})
  MakeMaterialAttributes(BaseColor=@n144.Base Color, Specular=@n146, Roughness=@n151, EmissiveColor=@n12, Normal=@n172)
  MaterialFunctionCall(
   In=@n208.Blended Material
   Color=ComponentMask(None=CollectionParameter() {collection=&MPC_Landscape, parameterName=Arid_Rock_01_Tint}) {b=true, g=true, r=true}
  ) {materialFunction=&MF_Athena_Material_BaseMultiply}
  MaterialFunctionCall(In=@n214, Contrast=Constant() {r=1}) {materialFunction=&CheapContrast}
  n215 = MaterialFunctionCall(Base Material=@n174.Result, Top Material=NamedRerouteUsage(), Alpha=LandscapeLayerSample() {parameterName=Ice}) {materialFunction=&MatLayerBlend_Standard}
  n216 = NamedRerouteUsage()
  n217 = NamedRerouteUsage()
  n218 = NamedRerouteUsage()
  n219 = NamedRerouteUsage()
  n220 = Saturate(None=Add(A=LandscapeLayerSample() {parameterName=Snow}, B=LandscapeLayerSample() {parameterName=Ice}))
  n221 = NamedRerouteUsage()
  n222 = ScalarParameter() {group=Snow, parameterName=Snow_HeightBlendContrast}
  MaterialFunctionCall(
   Base Material=@n174.Result
   Top Material=MaterialFunctionCall!(
    Material=MaterialFunctionCall(Base Material=@n215.Blended Material, Top Material=@n216, Alpha=@n217) {materialFunction=&MatLayerBlend_Standard}.Blended Material
    NewBaseColor=ShadingPathSwitch(Default=@n218, Mobile=@n219)
    Mask
   ) {materialFunction=/Engine/Functions/MaterialLayerFunctions/MatLayerBlend_MultiplyBaseColor}.Blended Material
   Alpha=Multiply(
    A=MaterialFunctionCall(Transition Phase=@n220, Height Texture=@n221, Contrast=@n222) {materialFunction=&HeightLerp}.Alpha
    B=NamedRerouteUsage()
   )
  ) {materialFunction=&MatLayerBlend_Standard}
  n223 = Multiply(A=ComponentMask(None=VertexNormalWS()) {b=true}, B=ScalarParameter() {defaultValue=1.1, group=Rock, parameterName=RockMultiply})
  n224 = OneMinus(None=Power(Base=@n223))
  n225 = LandscapeLayerSample() {parameterName=Arid_RockDetail, previewWeight=1}
  Max(A=@n224, B=Saturate(None=Multiply(A=@n225) {constB=2}))
  n226 = LandscapeLayerSample() {parameterName=Jungle_Sand, previewWeight=1}
  n227 = NamedRerouteDeclaration(None=LandscapeLayerSample() {parameterName=Arid}) {name=AridSand, nodeColor=<rgba 0.2188 0 1 1>}
  Max(A=@n226, B=@n227)
  n228 = Max(A=Max(A=NamedRerouteUsage(), B=NamedRerouteUsage()), B=NamedRerouteUsage())
  Max(A=@n228, B=LandscapeLayerSample() {parameterName=Jungle_Sand, previewWeight=1})
  n229 = NamedRerouteUsage()
  Max(A=NamedRerouteUsage(), B=@n229)
  n230 = NamedRerouteUsage()
  n231 = NamedRerouteUsage()
  n232 = Max(A=Max(A=Max(A=@n230, B=@n231), B=NamedRerouteUsage()), B=NamedRerouteUsage())
  Multiply(A=@n232) {constB=4}
  n233 = LandscapeLayerSample() {parameterName=Grass_Only, previewWeight=1}
  n234 = Add(
   A=Add(A=LandscapeLayerSample() {parameterName=Grass_Topology, previewWeight=1}, B=Round(None=@n233))
   B=Add(A=LandscapeLayerSample() {parameterName=Mountain_Grass}, B=LandscapeLayerSample() {parameterName=Farm_Grass})
  )
  Multiply(A=Multiply(A=@n204, B=Saturate(None=@n234)), B=@n205)
  n235 = MaterialFunctionCall!(Rock=@n26.Result) {materialFunction=/FigmentEnvironment/LM_8_1_1/Athena/Environments/Landscape/MaterialFunctions/Snow/MF_Athena_IceRock_01}
  n236 = GetMaterialAttributes(None=@n91.Material) {attributeGetTypes [#1]}
  n237 = ComponentMask(None=CollectionParameter() {collection=&MPC_Landscape, parameterName=Snow_CavTint}) {b=true, g=true, r=true}
  n238 = TextureSample(UVs=MaterialFunctionCall() {materialFunction=&MF_Athena_Terrain_Coords}.Result) {
   samplerSource=SSM_Wrap_WorldGroupSettings
   samplerType=SAMPLERTYPE_LinearColor
   texture=&T_Athena_Terrain_Topo_Mask
  }
  n239 = LinearInterpolate(A=Constant() {r=2}, B=Constant() {r=-3}, Alpha=Clamp(None=Multiply(A=@n238.G, B=@n238.G)))
  n240 = StaticSwitchParameter(
   True=MaterialFunctionCall(Material=@n91.Material, NewBaseColor=Multiply(A=@n236.Base Color, B=@n237), Mask=Clamp(None=@n239)) {materialFunction=&MatLayerBlend_MultiplyBaseColor}.Blended Material
   False=@n91.Material
  ) {parameterName=UseSnowCav}
  NamedRerouteDeclaration(None=MaterialFunctionCall(Base Material=@n235.Material, Top Material=@n240, Alpha=@n213.Alpha) {materialFunction=&MatLayerBlend_Standard}.Blended Material) {name=IceMA, nodeColor=<rgba 0.9453 0 1 1>}
  NamedRerouteDeclaration(None=@n91.Height) {name=HeightSnow, nodeColor=<rgba 0.2422 1 0 1>}
  NamedRerouteDeclaration(None=@n203) {name=SnowIce, nodeColor=<rgba 1 0.0469 0 1>}
  NamedRerouteDeclaration(None=@n100.Lerp Alpha No Contrast) {name=GravelMask, nodeColor=<rgba 0.6172 0 1 1>}
  NamedRerouteDeclaration(None=@n87.Alpha) {name=mudmask, nodeColor=<rgba 0 1 0.2969 1>}
  NamedRerouteDeclaration(None=@n78.Lerp Alpha No Contrast) {name=forestmask, nodeColor=<rgba 1 0 0.0234 1>}
  NamedRerouteDeclaration(None=@n32.Alpha) {name=SlopeMask_NoSnow, nodeColor=<rgba 0 1 0.7891 1>}
  n241 = LandscapeLayerSample() {parameterName=Grass_Only, previewWeight=1}
  n242 = LandscapeLayerSample() {parameterName=Grass_Topology, previewWeight=1}
  n243 = Max(
   A=Max(A=Max(A=@n241, B=@n242), B=LandscapeLayerSample() {parameterName=Golf_Fairway, previewWeight=1})
   B=LandscapeLayerSample() {parameterName=Farm_Grass}
  )
  n244 = LandscapeLayerSample() {parameterName=Mountain_Grass}
  n245 = Saturate(None=Max(A=Max(A=@n243, B=@n244), B=LandscapeLayerSample() {parameterName=Flowers}))
  n246 = Saturate(None=Max(A=@n228, B=NamedRerouteUsage()))
  NamedRerouteDeclaration(None=Multiply(A=NamedRerouteDeclaration(None=@n245) {name=Grass, nodeColor=<rgba 0 1 0.2969 1>}, B=OneMinus(None=@n246))) {name=FinalGrass, nodeColor=<rgba 0 1 0.6016 1>}
  n247 = Max(A=LandscapeLayerSample() {parameterName=Gravel, previewWeight=1}, B=LandscapeLayerSample() {parameterName=Mud, previewWeight=1})
  n248 = LandscapeLayerSample() {parameterName=Forest}
  n249 = Saturate(None=Saturate(None=Max(A=@n247, B=@n248)))
  n250 = NamedRerouteUsage()
  NamedRerouteDeclaration(None=Multiply(A=NamedRerouteDeclaration(None=@n249) {name=Dirt, nodeColor=<rgba 1 0 0.375 1>}, B=OneMinus(None=@n250))) {name=FinalDirt, nodeColor=<rgba 0 0.6953 1 1>}
  NamedRerouteDeclaration(None=@n92.Alpha) {name=RoadAlpha, nodeColor=<rgba 0 1 0.2969 1>}
  n251 = Multiply(A=LandscapeLayerSample() {parameterName=Road, previewWeight=1}) {constB=1.5}
  n252 = Saturate(None=@n229)
  NamedRerouteDeclaration(None=Subtract(A=NamedRerouteDeclaration(None=@n251) {name=Road, nodeColor=<rgba 0.4531 0 1 1>}, B=OneMinus(None=@n252))) {name=FinalRoad, nodeColor=<rgba 0 1 0.1797 1>}
  NamedRerouteDeclaration(None=Saturate(None=@n224)) {name=RockBlend, nodeColor=<rgba 1 0.0938 0 1>}
  n253 = OneMinus(None=Saturate(None=@n232))
  NamedRerouteDeclaration(None=Max(A=@n226, B=Multiply(A=@n227, B=@n253))) {name=FinalSand, nodeColor=<rgba 0.0312 1 0 1>}
  n254 = Max(A=LandscapeLayerSample() {parameterName=Snow}, B=LandscapeLayerSample() {parameterName=Ice})
  n255 = Max(A=NamedRerouteUsage(), B=NamedRerouteUsage())
  n256 = NamedRerouteUsage()
  n257 = Saturate(None=Multiply(A=Max(A=@n255, B=@n256)) {constB=4})
  NamedRerouteDeclaration(None=Multiply(A=NamedRerouteDeclaration(None=@n254) {name=Snow, nodeColor=<rgba 1 0 0.9141 1>}, B=OneMinus(None=@n257))) {name=SnowFinal, nodeColor=<rgba 0 1 0.7656 1>}
  NamedRerouteDeclaration(None=@n30.Result) {name=MAT_Grass, nodeColor=<rgba 1 0 0.2109 1>}
  NamedRerouteDeclaration(None=@n235.Material) {name=IceRock, nodeColor=<rgba 1 0 0.3984 1>}
  n258 = Max(
   A=Max(A=NamedRerouteUsage(), B=NamedRerouteUsage())
   B=Multiply(A=NamedRerouteUsage(), B=ScalarParameter() {defaultValue=0.4444, group=Snow, parameterName=GravelReveal})
  )
  n259 = Multiply(A=NamedRerouteUsage(), B=ScalarParameter() {defaultValue=1, group=Snow, parameterName=RoadReveal}) {constB=0.666}
  n260 = OneMinus(None=Saturate(None=Max(A=@n258, B=@n259)))
  n261 = ScalarParameter() {defaultValue=1, parameterName=SnowBlendPower}
  NamedRerouteDeclaration(None=Saturate(None=Power(Base=@n260, Exp=@n261))) {name=Reveal, nodeColor=<rgba 1 0.1406 0 1>}
  NamedRerouteDeclaration(None=@n30.Height) {name=MAT_Grass_Height, nodeColor=<rgba 1 0 0.2109 1>}
  n262 = LinearInterpolate(
   A=ScalarParameter() {parameterName=NonGrassSpecular}
   B=ScalarParameter() {defaultValue=0.5, parameterName=RoadSpecular}
   Alpha=LandscapeLayerSample() {parameterName=Road, previewWeight=1}
  )
  NamedRerouteDeclaration(None=LinearInterpolate(A=LinearInterpolate(A=@n74.Specular, B=@n262, Alpha=@n148), B=Constant() {desc=Jungle Sand, r=0.1}, Alpha=NamedRerouteUsage()) {constB=0.1}) {name=Spec, nodeColor=<rgba 1 0.9609 0 1>}
  NamedRerouteDeclaration(None=@n44) {name=MobileColorMap, nodeColor=<rgba 0.5938 1 0 1>}
  NamedRerouteDeclaration(None=Multiply(A=OneMinus(None=@n235.EmissiveMask), B=OneMinus(None=@n213.Alpha))) {name=WinterEmissiveMask, nodeColor=<rgba 0.8516 1 0 1>}
  NamedRerouteDeclaration(None=@n213.Alpha_ColorMap) {name=SlopeMask_AlphaColorMap, nodeColor=<rgba 0 1 0.5781 1>}
  NamedRerouteDeclaration(None=@MaterialExpressionStaticSwitchParameter_7) {name=MAT_Arid, nodeColor=<rgba 1 0.2578 0 1>}
  NamedRerouteDeclaration(None=@n14.Alpha) {name=MASK_JungleSand, nodeColor=<rgba 0.5 1 0 1>}
  NamedRerouteDeclaration(None=MaterialFunctionCall(Base Material=@n26.Result, Top Material=@n240, Alpha=@n32.Alpha) {materialFunction=&MatLayerBlend_Standard}.Blended Material) {name=IceSnowSlopeMA, nodeColor=<rgba 1 0 0.6562 1>}
  NamedRerouteDeclaration(None=@n91.Material) {name=SnowMA, nodeColor=<rgba 1 0 0.0234 1>}
  NamedRerouteDeclaration(None=@n213.Alpha) {name=SlopeMask, nodeColor=<rgba 0 0.3437 1 1>}
  NamedRerouteDeclaration(None=@n20) {name=BaseColorMap, nodeColor=<rgba 1 0.3984 0 1>}
  NamedRerouteDeclaration(None=@n33) {name=GrassRock, nodeColor=<rgba 0 1 0.0391 1>}
  NamedRerouteDeclaration(None=@n173) {name=GrassMA, nodeColor=<rgba 0 0.0625 1 1>}
  NamedRerouteUsage()
  NamedRerouteUsage()
  NamedRerouteUsage()
  NamedRerouteUsage()
  NamedRerouteUsage()
  NamedRerouteUsage()
  NamedRerouteUsage()
  OneMinus(None=@n184)
  Reroute(None=@n182.Result)
  Reroute(None=@n192) {desc=Grass}
  n263 = LandscapeLayerSample() {parameterName=Farm_Grass}
  n264 = LandscapeLayerSample() {parameterName=Mountain_Grass}
  n265 = Add(A=Add(A=@n175, B=Subtract(A=@n263, B=@n191)), B=Subtract(A=Add(A=@n264, B=@n263), B=@n191))
  n266 = Subtract(A=Subtract(A=Add(A=@n265, B=@n204), B=@n190), B=@n189)
  n267 = Ceil(None=LandscapeLayerSample() {parameterName=Jungle_Pools, previewWeight=1})
  Reroute(None=Clamp(None=Subtract(A=@n266, B=@n267))) {desc=bushes}
  Reroute(None=@n214)
  Round(None=@n184)
  RuntimeVirtualTextureOutput!(
   BaseColor=@n144.Base Color
   Specular=@n146
   Roughness=@n151
   Normal=Transform(None=@n172)
   WorldHeight=WorldPosition().Z
   Opacity
   Mask
   Displacement
   Mask4
  )
  Saturate(None=@n251)
  Saturate(None=@n223)
  ScalarParameter()
  ScalarParameter() {defaultValue=306200, parameterName=TerrainSize}
  SetMaterialAttributes(Normal=Transform(None=VertexNormalWS()) {transformSourceType=TRANSFORMSOURCE_World, transformType=TRANSFORM_Tangent}) {attributeSetTypes [#5]}
  Subtract(A=LandscapeLayerSample() {parameterName=Farm_Grass}, B=@n191)
  TextureSample(UVs=@n17.Result) {
   samplerSource=SSM_Wrap_WorldGroupSettings
   samplerType=SAMPLERTYPE_LinearColor
   texture=/FigmentEnvironment/LM_8_1_1/Athena/Environments/Landscape/Textures/T_Athena_Terrain_GroundColor
  }
 ]
}
